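# UI material for the Ensemble item shop background, drawing an animated red backdrop with speed lines and spinning shurikens.
# Radial and linear gradients, an edge falloff mask and a halftone function blend into the base color layers.
# Speed lines come from a panned sparkle noise texture sampled in radial coordinates.
# Two offset shuriken samples rotate over time and loop diagonally via Frac, feeding a named reroute for compositing.
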Material /ItemShop/UI/Art/Ensemble/M_UI_EnsembleShopBackground
graph {
 properties {bUsedWithStaticMesh=false, materialDomain=MD_UI, pixelDepthOffsetMode=PDOM_Legacy, usedShadingModels=Default Lit}
 nodes [
  Add()
  n1 = Constant3Vector() {constant=<rgba 1 0.0257 0 0>}
  n2 = MaterialFunctionCall(UV=TextureCoordinate(), GradientStart=Constant() {r=-0.4237}) {materialFunction=&MF_UI_RadialGradient}
  n3 = TextureCoordinate()
  n4 = ComponentMask(None=Abs(None=ConstantBiasScale(None=@n3) {bias=-0.5, scale=2})) {g=true}
  n5 = ScalarParameter() {defaultValue=1.0389, parameterName=Subtract}
  n6 = Saturate(None=Multiply(A=Subtract(A=@n4, B=@n5), B=ScalarParameter() {defaultValue=-0.1421, parameterName=Falloff}))
  n7 = ScalarParameter() {defaultValue=9.7133, parameterName=Intensity}
  n8 = MaterialFunctionCall() {materialFunction=&/MF_UI_Halftone}
  n9 = Constant() {r=2.0385}
  n10 = VectorParameter() {parameterName=RGB2, defaultValue=<rgba 0 0 0 1>}
  n11 = VectorParameter() {parameterName=RGB4, defaultValue=<rgba 0.6724 0.6724 0.6724 1>}
  n12 = MaterialFunctionCall(
   GradientStart=ScalarParameter() {parameterName=GradientStart}
   GradientEnd=ScalarParameter() {defaultValue=0.7566, parameterName=GradientEnd}
   RotationAngle=Constant() {r=42.2274}
   GradientOffset=Constant() {r=0.248}
  ) {materialFunction=&MF_UI_LinearGradient}
  n13 = LinearInterpolate(
   A=Multiply(A=Multiply(A=@n6, B=@n7), B=VectorParameter() {parameterName=RGB3, defaultValue=<rgba 0.6724 0.0091 0.004 1>}.RGB)
   B=LinearInterpolate(
    A=MaterialFunctionCall(UVs=@n8.Result, Texture Scale=@n9) {materialFunction=&ScaleUVsByCenter}.UVs
    B=LinearInterpolate(A=@n10.RGB, B=@n11.RGB, Alpha=@n12.Result)
   ) {constA=-0.008, constAlpha=1.1215}
  ) {constAlpha=0.5181}
  n14 = Constant()
  n15 = ScalarParameter() {defaultValue=0.085, parameterName=PrimaryYShift}
  n16 = Add(A=Add(A=TextureCoordinate(), B=AppendVector(A=@n14, B=@n15))) {constB=-0.5}
  n17 = Panner(
   Coordinate=Multiply(
    A=MaterialFunctionCall(Vector or UVs=@n16) {materialFunction=&VectorToRadialValue}.Radial Coordinates
    B=Constant2Vector() {g=0.035, r=6}
   )
   Speed=AppendVector(A=Constant(), B=ScalarParameter() {defaultValue=-0.009, parameterName=SpeedLinesSpeed})
  ) {speedY=0.075}
  n18 = Add(A=Constant2Vector() {g=0.5, r=0.1}, B=@n17)
  n19 = Multiply(
   A=Constant3Vector() {constant=<rgba 1 0.0257 0 0>}
   B=Add(
    A=TextureSample(UVs=@n18) {samplerType=SAMPLERTYPE_LinearColor, texture=/Game/Gadgets/Assets/Slow_Field/Textures/T_Slowfield_SparkleNoise_01}.R
    B=TextureSample(UVs=@n17) {samplerType=SAMPLERTYPE_LinearColor, texture=&T_Slowfield_SparkleNoise_01}.G
   )
  )
  LinearInterpolate(
   A=LinearInterpolate(A=Multiply(A=@n1, B=@n2.Smooth), B=LinearInterpolate(A=@n13, B=@n19) {constAlpha=0.532}) {constAlpha=-0.372}
   B=Constant() {r=1.4302}
   Alpha=NamedRerouteUsage()
  ) {constAlpha=1.2031, constB=4.2714}
  n20 = VectorParameter() {parameterName=RGB1, defaultValue=<rgba 0.6724 0.0091 0.004 1>}
  n21 = Multiply(A=Time(), B=ScalarParameter() {defaultValue=0.16, parameterName=Speed})
  n22 = OneMinus(None=@n21)
  n23 = Multiply(A=AppendVector(A=Frac(None=@n22), B=Frac(None=@n21)), B=Constant() {r=2.5})
  n24 = ConstantBiasScale(None=TextureCoordinate()) {bias=-0.5, scale=2}
  n25 = MaterialFunctionCall(
   UVs=Add(A=Add(A=@n23, B=@n24), B=Constant2Vector() {g=-0.6795, r=-0.5647})
   "Rotation Angle (0-1)"=Multiply(A=Time(), B=ScalarParameter() {defaultValue=2.0158, parameterName=ShurikenRotationSpeed})
  ) {materialFunction=&CustomRotator}
  n26 = TextureSample(UVs=MaterialFunctionCall(
    UVs=ConstantBiasScale(None=@n25.Rotated Values) {bias=0.5}
    Texture Scale=ScalarParameter() {defaultValue=0.092, parameterName=ShurikenTextureSize}
   ) {materialFunction=&ScaleUVsByCenter}.UVs) {samplerSource=SSM_Clamp_WorldGroupSettings, texture=&/T_UI_Ensemble_Shuriken}
  n27 = Multiply(A=Time(), B=ScalarParameter() {defaultValue=0.16, parameterName=Speed})
  n28 = OneMinus(None=@n27)
  n29 = Multiply(A=AppendVector(A=Frac(None=@n28), B=Frac(None=@n27)), B=Constant() {r=2.5})
  n30 = ConstantBiasScale(None=TextureCoordinate()) {bias=-0.5, scale=2}
  n31 = MaterialFunctionCall(
   UVs=Add(A=Add(A=@n29, B=@n30), B=Constant2Vector() {g=-1.3, r=-0.9})
   "Rotation Angle (0-1)"=Multiply(A=Time(), B=ScalarParameter() {defaultValue=2.0158, parameterName=ShurikenRotationSpeed})
  ) {materialFunction=&CustomRotator}
  n32 = TextureSample(UVs=MaterialFunctionCall(
    UVs=ConstantBiasScale(None=@n31.Rotated Values) {bias=0.5}
    Texture Scale=ScalarParameter() {defaultValue=0.092, parameterName=ShurikenTextureSize}
   ) {materialFunction=&ScaleUVsByCenter}.UVs) {samplerSource=SSM_Clamp_WorldGroupSettings, texture=&T_UI_Ensemble_Shuriken}
  NamedRerouteDeclaration(None=Add(A=Multiply(A=@n20.RGB, B=@n26.RGB), B=Multiply(A=@n20.RGB, B=@n32.RGB))) {name=Shuriken_Anim, nodeColor=<rgba 0 0 0 1>}
  TextureCoordinate()
  n33 = TextureCoordinate()
  n34 = Rotator(Coordinate=@n33, Time=Constant() {r=1})
  TextureSample(UVs=AppendVector(A=ComponentMask(None=@n34) {r=true}, B=ComponentMask(None=@n33) {g=true})) {texture=&/T_UI_Halftone}
 ]
}
